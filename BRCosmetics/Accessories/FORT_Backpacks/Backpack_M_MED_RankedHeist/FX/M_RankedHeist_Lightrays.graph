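# Translucent, unlit, two-sided light ray material for the RankedHeist backpack's Niagara mesh particles.
# Opposed panning line textures, sampled at three U offsets, build chromatic RGB streaks.
# Vertex blue blends inner and outer color, alpha and speed from particle data, though tiling and WPO offset lerps use only Outer values.
# UV gradient, depth, camera, view-angle and optional edge fades shape opacity; WPO pushes vertices along the tangent-normal cross product.
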Material /BRCosmetics/Accessories/FORT_Backpacks/Backpack_M_MED_RankedHeist/FX/M_RankedHeist_Lightrays
graph {
 properties {
  bUsedWithNiagaraMeshParticles=true
  blendMode=BLEND_Translucent
  pixelDepthOffsetMode=PDOM_Legacy
  shadingModel=MSM_Unlit
  twoSided=true
  usedShadingModels=Unlit
 }
 nodes [
  n1 = ComponentMask(None=TextureCoordinate()) {g=true}
  n2 = OneMinus(None=@n1)
  n3 = NamedRerouteUsage()
  n4 = Power(
   Base=@n2
   Exp=LinearInterpolate(A=ScalarParameter() {defaultValue=3, parameterName=A}, B=ScalarParameter() {defaultValue=2, parameterName=B}, Alpha=Desaturation(None=@n3)) {constA=2, constB=6}
  )
  n5 = ShadingPathSwitch(Default=DepthFade(FadeDistance=ScalarParameter() {defaultValue=75, parameterName=Depth Fade Distance}), Mobile=Constant() {r=1})
  n6 = MaterialFunctionCall(Distance=ScalarParameter() {defaultValue=150, parameterName=NearCamFade_Distance}) {materialFunction=&Near_Camera_Fade}
  n7 = ScalarParameter() {defaultValue=1.5, parameterName=Viewing Angle Fade Sensitivity}
  n8 = Multiply(
   A=Multiply(A=Multiply(A=@n2, B=@n4), B=NamedRerouteUsage())
   B=Multiply(
    A=Multiply(A=@n5, B=@n6.Result)
    B=MaterialFunctionCall("FadeSensitivity (optional)"=@n7) {materialFunction=&FadeBasedOnViewAngle}.Result
   )
  )
  n9 = Power(Base=@n1, Exp=ScalarParameter() {defaultValue=1, group=EdgeFade, parameterName=Edge Fade Amount}) {constExponent=1.1}
  LinearInterpolate(A=@n8, B=Multiply(A=SmoothStep(Value=@n9) {constMax=0.2}, B=@n8), Alpha=ScalarParameter() {group=EdgeFade, parameterName=Add Edge Fade})
  n10 = Multiply(A=NamedRerouteUsage(), B=NamedRerouteUsage())
  n11 = TextureCoordinate()
  n12 = Saturate(None=Power(Base=ComponentMask(None=@n11) {g=true}, Exp=ScalarParameter() {defaultValue=1.372}))
  MaterialFunctionCall(Emissive=MaterialFunctionCall(LDR Input=Saturate(None=@n10), HDR Tint=@n10, HDR Intensity Multiplier=LinearInterpolate(Alpha=@n12) {constA=1, constB=0.25}) {materialFunction=&DeriveHDRfromLDR}.RemappedRange) {materialFunction=&MF_FX_EmissiveAdjustments}
  n13 = TextureCoordinate()
  n14 = ScalarParameter() {defaultValue=2.7817, parameterName=A_WPO}
  n15 = ScalarParameter() {defaultValue=1.4581, parameterName=WPO_B}
  n16 = VertexColor()
  n17 = DynamicParameter() {
   parameterIndex=2
   defaultValue=<rgba 0 1 1 1>
   paramNames ["Random 0-1 (Inner)", "Panner Speed (Inner)", "WPO Offset (Outer)", "WPO Offset (Inner)"]
  }
  n18 = VertexColor()
  n19 = VertexTangentWS()
  n20 = VertexNormalWS()
  Multiply(
   A=Power(Base=ComponentMask(None=@n13) {g=true}, Exp=LinearInterpolate(A=@n14, B=@n15, Alpha=@n16.B) {constA=2, constB=6})
   B=Multiply(A=LinearInterpolate(A=@n17."WPO Offset (Outer)", B=@n17."WPO Offset (Outer)", Alpha=@n18.B), B=CrossProduct(A=@n19, B=@n20))
  )
  n21 = Time()
  n22 = DynamicParameter() {defaultValue=<rgba 0 1 1 1>, paramNames [Random 0-1, Panner Speed, U Tile Outer, U Tile Inner]}
  n23 = DynamicParameter() {
   parameterIndex=2
   defaultValue=<rgba 0 1 1 1>
   paramNames ["Random 0-1 (Inner)", "Panner Speed (Inner)", "WPO Offset (Outer)", "WPO Offset (Inner)"]
  }
  n24 = LinearInterpolate(A=@n22.Random 0-1, B=@n23."Random 0-1 (Inner)")
  n25 = VertexColor()
  NamedRerouteDeclaration(None=Multiply(A=Add(A=@n21, B=@n24), B=LinearInterpolate(A=@n22.Panner Speed, B=@n23."Panner Speed (Inner)", Alpha=@n25.B))) {name=Panner Time, nodeColor=<rgba 0 1 0.5781 1>}
  n26 = ParticleColor()
  n27 = DynamicParameter() {parameterIndex=1, paramNames [R,G,B,A]}
  n28 = VertexColor()
  NamedRerouteDeclaration(None=LinearInterpolate(A=@n26.RGB, B=@n27.RGB, Alpha=@n28.B)) {name=Particle Color, nodeColor=<rgba 0 1 0.2969 1>}
  n29 = NamedRerouteUsage()
  n30 = AppendVector(A=ScalarParameter() {defaultValue=0.025, parameterName=Chromatic Offset}, B=Constant())
  n31 = TextureObjectParameter() {parameterName="Lines Texture (RG)", texture=/Game/Effects/Fort_Effects/Textures/Outlander/T_InTheZone_LightRays_01}
  n32 = TextureSample(UVs=Panner(Coordinate=Add(A=@n29, B=@n30), Time=NamedRerouteUsage()) {speedX=0.0702}, Tex=@n31) {texture=&T_InTheZone_LightRays_01}
  n33 = Multiply(A=NamedRerouteUsage(), B=ScalarParameter() {defaultValue=1.25, parameterName=Chroma U Offset})
  n34 = TextureSample(UVs=Panner(Coordinate=Add(A=@n33, B=@n30), Time=NamedRerouteUsage()) {speedX=-0.047}, Tex=@n31) {texture=&T_InTheZone_LightRays_01}
  n35 = Multiply(A=Add(A=@n32.R, B=@n34.R), B=Add(A=@n32.G, B=@n34.G))
  n36 = TextureObjectParameter() {parameterName="Lines Texture (RG)", texture=&T_InTheZone_LightRays_01}
  n37 = TextureSample(UVs=Panner(Coordinate=NamedRerouteUsage(), Time=NamedRerouteUsage()) {speedX=0.0702}, Tex=@n36) {texture=&T_InTheZone_LightRays_01}
  n38 = NamedRerouteUsage()
  n39 = ScalarParameter() {defaultValue=1.25, parameterName=Chroma U Offset}
  n40 = TextureSample(UVs=Panner(Coordinate=Multiply(A=@n38, B=@n39), Time=NamedRerouteUsage()) {speedX=-0.047}, Tex=@n36) {texture=&T_InTheZone_LightRays_01}
  n41 = Multiply(A=Add(A=@n37.R, B=@n40.R), B=Add(A=@n37.G, B=@n40.G))
  n42 = NamedRerouteUsage()
  n43 = Multiply(A=AppendVector(A=ScalarParameter() {defaultValue=0.025, parameterName=Chromatic Offset}, B=Constant())) {constB=-1}
  n44 = TextureObjectParameter() {parameterName="Lines Texture (RG)", texture=&T_InTheZone_LightRays_01}
  n45 = TextureSample(UVs=Panner(Coordinate=Add(A=@n42, B=@n43), Time=NamedRerouteUsage()) {speedX=0.0702}, Tex=@n44) {texture=&T_InTheZone_LightRays_01}
  n46 = Multiply(A=NamedRerouteUsage(), B=ScalarParameter() {defaultValue=1.25, parameterName=Chroma U Offset})
  n47 = TextureSample(UVs=Panner(Coordinate=Add(A=@n46, B=@n43), Time=NamedRerouteUsage()) {speedX=-0.047}, Tex=@n44) {texture=&T_InTheZone_LightRays_01}
  n48 = Multiply(A=Add(A=@n45.R, B=@n47.R), B=Add(A=@n45.G, B=@n47.G))
  NamedRerouteDeclaration(None=Saturate(None=MaterialFunctionCall(R=@n35, G=@n41, B=@n48) {materialFunction=&AppendMany}.RGB)) {name=Chromatic Texture Setup, nodeColor=<rgba 0.9413 0.4666 1 1>}
  NamedRerouteDeclaration(None=LinearInterpolate(A=@n26.A, B=@n27.A, Alpha=@n28.B)) {name=Particle Alpha, nodeColor=<rgba 0.9453 1 0 1>}
  n49 = DynamicParameter() {defaultValue=<rgba 0 1 1 1>, paramNames [Random 0-1, Panner Speed, U Tile Outer, U Tile Inner]}
  n50 = VertexColor()
  NamedRerouteDeclaration(None=Multiply(A=TextureCoordinate(), B=LinearInterpolate(A=@n49.U Tile Outer, B=@n49.U Tile Outer, Alpha=@n50.B))) {name=Chroma UVs, nodeColor=<rgba 0 0.9531 1 1>}
 ]
}
